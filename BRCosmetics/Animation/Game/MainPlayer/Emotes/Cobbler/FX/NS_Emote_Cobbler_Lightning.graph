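# Niagara lightning effect for the Cobbler player emote, built from three emitters placed at a shared system offset.
# Thunder spawns jittering bolt sprites on a cylinder, pushed outward with drag; spawn group picks each sprite's size and SubUV frame.
# Decal_Glow_And_Light adds a shrinking glow with a pulsing light, and Black_Line a single expanding, fading sprite.
# Thunder's size and velocity take a front-end scale modifier.
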NiagaraSystem /BRCosmetics/Animation/Game/MainPlayer/Emotes/Cobbler/FX/NS_Emote_Cobbler_Lightning
dependencies {
 usedDataInterfaces [
  &NiagaraDataInterfaceCurve
  &NiagaraDataInterfaceArrayFloat2
  &NiagaraDataInterfaceMeshRendererInfo
  &NiagaraDataInterfaceSpriteRendererInfo
  &NiagaraDataInterfaceColorCurve
 ]
 usedDynamicInputs []
 usedModules [
  &self:SystemSpawnScript.SystemScriptSource.SystemScriptGraph.NiagaraNodeAssignment_0.SetVariables_#1
  &self:SystemSpawnScript.SystemScriptSource.SystemScriptGraph.NiagaraNodeAssignment_1.SetVariables_#2
  &SystemState
 ]
}
emitters {
 Black_Line {
  emitterUpdateScript [
   &EmitterState(Life Cycle Mode=::NewEnumerator1, Loop Behavior=::NewEnumerator1, Loop Duration=1, UseLoopDelay=false)
   &SpawnBurst_Instantaneous(Spawn Count=1, Spawn Time=0.05, Use Spawn Probability=false)
  ]
  particleSpawnScript [&InitializeParticle(
    Color=<rgba 1 1 1 25>
    Lifetime=0.25
    Position Offset=$System.LocalOffset
    Sprite Rotation Mode=::NewEnumerator2
    Sprite Size Mode=::NewEnumerator3
    Uniform Sprite Size=120
    UsePositionOffset=true
    Write Material Random=true
   ) {module=InitializeParticle001}]
  particleUpdateScript [
   &ParticleState()
   &SolveForcesAndVelocity(Clamp Velocity=false, Limit Acceleration=false)
   &ScaleSpriteSize(Scale Factor=&Vector2DFromFloat(Value=&FloatFromCurve(FloatCurve=DataInterface() {propertyValues {Curve {keys [
          {arriveTangent=1.6807, interpMode=RCIM_Cubic, leaveTangent=1.6807, tangentMode=RCTM_User, time=0, value=0}
          {interpMode=RCIM_Cubic, tangentMode=RCTM_User, time=1, value=1}
         ]}}})))
   &DynamicMaterialParameters(
    Index 0 Param 1=2
    Module.Param0WriteEnabled=true
    Module.Param1WriteEnabled=true
    Module.Param2WriteEnabled=true
    Module.Param3WriteEnabled=true
   )
   &ScaleColor(
    Scale Alpha=&FloatFromCurve(
     FloatCurve=DataInterface() {propertyValues {Curve {keys [
         {arriveTangent=-13.9062, interpMode=RCIM_Cubic, leaveTangent=-13.9062, tangentMode=RCTM_User, time=0, value=7}
         {arriveTangent=-7, interpMode=RCIM_Cubic, leaveTangent=-7, time=0.2488, value=3}
         <at 1 0 RCIM_Cubic>
        ]}}}
     Scale Curve=3
    )
    ScaleA=true
    ScaleRGB=true
   )
   &CameraOffset()
  ]
  rendererClasses [&NiagaraSpriteRendererProperties]
  renderers [{rendererClass=&NiagaraSpriteRendererProperties}]
 }
 Decal_Glow_And_Light {
  emitterUpdateScript [
   &EmitterState(Life Cycle Mode=::NewEnumerator1, Loop Behavior=::NewEnumerator1, Loop Duration=1, UseLoopDelay=false)
   &SpawnBurst_Instantaneous(Spawn Time=0.05, Use Spawn Probability=false, UseLoopCountLimit=false) {module=SpawnBurst_Instantaneous001}
  ]
  particleSpawnScript [&InitializeParticle(
    Color=<rgba 0.1 0.1 0.1 0.1>
    Lifetime=0.2
    Position Offset=$System.LocalOffset
    Sprite Size Mode=::NewEnumerator3
    Uniform Sprite Size=100
    UsePositionOffset=true
    Write Material Random=true
   ) {module=InitializeParticle001}]
  particleUpdateScript [
   &ParticleState()
   &ScaleSpriteSize(Uniform Curve Sprite Scale=DataInterface() {propertyValues {Curve {keys [{interpMode=RCIM_Cubic, tangentMode=RCTM_User, time=0, value=1}, {interpMode=RCIM_Cubic, tangentMode=RCTM_User, time=1, value=0}]}}})
   &FN_Light_Attributes(
    Color=&MultiplyLinearColor(
     Linear Color=$Particles.Color
     Scale Factor=&MakeVector4_FromFloat(Float=&FloatFromCurve(FloatCurve=DataInterface() {propertyValues {Curve {keys [
           {interpMode=RCIM_Cubic, tangentMode=RCTM_User, time=0, value=0}
           {interpMode=RCIM_Cubic, tangentMode=RCTM_User, time=0.5, value=1}
           {tangentMode=RCTM_User, time=1, value=0}
          ]}}}))
    )
    Create Color Attribute=true
    Create Position Attribute=true
    Create Radius Attribute=true
    Position=&AddVectorToPosition(Position=$Particles.Position, Vector=<0 5 0>)
    Radius=30
    Scale Color=5
   )
   /SetVariables(Particles.LightColor=&MultiplyLinearColor(
     Linear Color=$Particles.Color
     Scale Factor=&MakeVector4_FromFloat(Float=&FloatFromCurve(FloatCurve=DataInterface() {propertyValues {Curve {keys [
           {interpMode=RCIM_Cubic, tangentMode=RCTM_User, time=0, value=0}
           {interpMode=RCIM_Cubic, tangentMode=RCTM_User, time=0.5, value=1}
           {tangentMode=RCTM_User, time=1, value=0}
          ]}}}))
    )) {isSetParameters=true}
   &CameraOffset(Camera Offset Amount=-25)
   &ScaleColor(
    Scale Alpha=&FloatFromCurve(FloatCurve=DataInterface() {propertyValues {Curve {keys [{interpMode=RCIM_Cubic, tangentMode=RCTM_User, time=0, value=1}, {interpMode=RCIM_Cubic, tangentMode=RCTM_User, time=1, value=0}]}}})
    Scale RGB=&VectorFromFloat(Value=&FloatFromCurve(
      FloatCurve=DataInterface() {propertyValues {Curve {keys [{interpMode=RCIM_Cubic, tangentMode=RCTM_User, time=0, value=1}, {interpMode=RCIM_Cubic, tangentMode=RCTM_User, time=1, value=0}]}}}
      Scale Curve=5
     ))
    ScaleA=true
    ScaleRGB=true
   )
  ]
  rendererClasses [&NiagaraLightRendererProperties, &NiagaraSpriteRendererProperties]
  renderers [
   {rendererClass=&NiagaraLightRendererProperties, rendererIndex=0}
   {rendererClass=&NiagaraSpriteRendererProperties, rendererIndex=1}
  ]
 }
 Thunder {
  emitterUpdateScript [
   &EmitterState(Life Cycle Mode=::NewEnumerator1, Loop Behavior=::NewEnumerator1, Loop Delay=0.05, Loop Duration=0.4, UseLoopDelay=true)
   &SpawnBurst_Instantaneous(Spawn Count=&UniformRangedInt(Minimum=1, Maximum=1), Spawn Time=0.05, Use Spawn Probability=false)
   &SpawnBurst_Instantaneous(
    Spawn Count=&UniformRangedInt(Minimum=1, Maximum=2)
    Spawn Group=1
    Spawn Time=0.05
    Use Spawn Probability=false
   ) {module=SpawnBurst_Instantaneous001}
   &SpawnBurst_Instantaneous(
    Spawn Count=&UniformRangedInt(Minimum=1, Maximum=2)
    Spawn Group=2
    Spawn Time=0.05
    Use Spawn Probability=false
   ) {module=SpawnBurst_Instantaneous002}
   &SpawnBurst_Instantaneous(
    Spawn Count=&UniformRangedInt(Minimum=1, Maximum=1)
    Spawn Group=3
    Spawn Time=0.05
    Use Spawn Probability=false
   ) {module=SpawnBurst_Instantaneous003}
   &SpawnRate(
    SpawnGroup:NiagaraInt32=&RandomRangeInt(Minimum=0, Maximum=3, Evaluation Type=::NewEnumerator1)
    SpawnRate=10
    Use Spawn Probability=true
   )
  ]
  particleSpawnScript [
   &InitializeParticle(
    Lifetime=&RandomRangeFloat(Minimum=0.4, Maximum=0.8)
    Position Offset=$System.LocalOffset
    Sprite Size=&Multiply_Vector2D_ByFloat(
     Float=&Multiply_Float(
      A=&RandomRangeFloat(Minimum=1, Maximum=1.25)
      B=&FN_FrontEndFloatModifier(Gameplay=1, Front End=$System.FrontendScale)
     )
     Vector2D=&SelectVector2DFromArray(
      Direct Array Index=&SpawnGroup()
      Vector 2D Selection Array=DataInterface() {propertyValues {FloatData [<20 40>, <20 20>, <15 20>, <20 40>]}}
     )
    )
    Sprite Size Mode=::NewEnumerator4
    UsePositionOffset=true
    Write Material Random=true
   )
   &ShapeLocation(
    Cylinder Distribution=::NewEnumerator2
    Cylinder Height=10
    Cylinder Height Midpoint=0
    Cylinder Radius=&RandomRangeFloat(Minimum=20, Maximum=30)
    MaskSpawnGroup=false
    Rotation Axis=<0 1 0>
    Rotation Mode=::NewEnumerator1
    Shape Origin=$Particles.Position
    Shape Primitive=::NewEnumerator1
    U Position=&Divide_Float(A=&MakeFloatFromInt(Int=&ReturnParticleIDAttributes()), B=10)
    UseOwnerScale=true
    V Position=&RandomRangeFloat(Minimum=0, Maximum=1)
   )
   &AddVelocity(
    Module.Use Falloff Distance=false
    Origin Offset=<0 -15 0>
    Override Seed=false
    Velocity Mode=::NewEnumerator1
    Velocity Origin:Vector3f=$System.LocalOffset
    Velocity Speed=&Multiply_Float(
     A=&RandomRangeFloat(Minimum=50, Maximum=60)
     B=&FN_FrontEndFloatModifier(Gameplay=1, Front End=$System.FrontendScale)
    )
   )
  ]
  particleUpdateScript [
   &ParticleState()
   &Drag(Drag=&UniformRangedFloat(Minimum=5, Maximum=6), Use Linear Drag=true, Use Rotational Drag=false)
   &SolveForcesAndVelocity(Clamp Velocity=false, Limit Acceleration=false)
   &JitterPosition(Jitter Amount:NiagaraFloat=&RandomRangeFloat(Minimum=3, Maximum=5), Jitter Delay=0.2)
   &SubUVAnimation(
    Direct Frame Index:NiagaraFloat=&MakeFloatFromInt(Int=&SpawnGroup())
    "Sprite Renderer "=DataInterface() {propertyValues {SpriteRenderer=&self:Thunder_0.NiagaraSpriteRendererProperties_1}}
    SubUV Animation Mode:ENiagara_SubUVLookupModeV2=::NewEnumerator5
   )
   &ScaleColor(
    Linear Color Curve=DataInterface() {propertyValues {AlphaCurve {keys [<at 0.8 1>,<at 1 0>]}, BlueCurve {keys [<at 0 6>]}, GreenCurve {keys [<at 0 6>]}, RedCurve {keys [<at 0 6>]}}}
    Scale Mode=::NewEnumerator2
   )
   &DynamicMaterialParameters(
    Index 0 Param 1=&FloatFromCurve(
     FloatCurve=DataInterface() {propertyValues {Curve {keys [
         {interpMode=RCIM_Cubic, tangentMode=RCTM_User, time=0.9, value=0}
         {arriveTangent=22.6971, interpMode=RCIM_Cubic, leaveTangent=22.6971, tangentMode=RCTM_User, time=1, value=1}
        ]}}}
     Scale Curve=0.3
    )
    Module.Param0WriteEnabled=true
    Module.Param1WriteEnabled=true
    Module.Param2WriteEnabled=true
    Module.Param3WriteEnabled=true
   )
  ]
  rendererClasses [&NiagaraSpriteRendererProperties]
  renderers [{rendererClass=&NiagaraSpriteRendererProperties}]
 }
}
summary {
 systemName=NS_Emote_Cobbler_Lightning
 emitters [
  {emitterName=Thunder, rendererClasses [&NiagaraSpriteRendererProperties]}
  {
   emitterName=Decal_Glow_And_Light
   rendererClasses [&NiagaraLightRendererProperties, &NiagaraSpriteRendererProperties]
  }
  {emitterName=Black_Line, rendererClasses [&NiagaraSpriteRendererProperties]}
 ]
 userVariables [
  User.bIsFrontEnd {type=NiagaraBool}
  User.bIsFrontEndPreview {type=NiagaraBool}
 ]
}
